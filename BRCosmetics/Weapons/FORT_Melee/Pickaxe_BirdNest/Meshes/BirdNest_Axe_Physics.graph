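# Physics asset for the BirdNest pickaxe, a two-handed melee cosmetic, defining collision and joint limits for its four dynamic fur bones.
# A single capsule covers the handle bone, and each dyn_fur bone gets its own small capsule.
# Each fur body is constrained directly to the handle rather than chained, with narrow swing and twist limits, slightly wider on the middle two, so the fur can sway without straying far.
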PhysicsAsset /BRCosmetics/Weapons/FORT_Melee/Pickaxe_BirdNest/Meshes/BirdNest_Axe_Physics
bodies {
 Melee_TwoHanded_Handle [{
   shapeType=Capsule
   length=29.6845
   radius=6.7687
   shapeName=Melee_TwoHanded_Handle_capsule
   center=<0.2997 -1.0863 17.6419>
   rotation=<rot 1.5544 -90 -0.275>
  }]
 dyn_fur_01 [{
   shapeType=Capsule
   length=3.5426
   radius=3.108
   shapeName=dyn_fur_01_capsule
   center=<0.8726 0.3875 0.9139>
   rotation=<rot 34.3159 -90 -13.6319>
  }]
 dyn_fur_02 [{
   shapeType=Capsule
   length=4.8232
   radius=4.166
   shapeName=dyn_fur_02_capsule
   center=<0.4129 -0.3854 -0.3588>
   rotation=<rot 0 -94.2393 -86.7598>
  }]
 dyn_fur_03 [{
   shapeType=Capsule
   length=4.7618
   radius=4.4213
   shapeName=dyn_fur_03_capsule
   center=<0.3239 -0.1073 -0.4017>
   rotation=<rot 0 -95.7584 -88.1511>
  }]
 dyn_fur_04 [{
   shapeType=Capsule
   length=4.5425
   radius=3.6946
   shapeName=dyn_fur_04_capsule
   center=<-0.1227 -0.2159 0.0653>
   rotation=<rot 0 94.6873 -48.7441>
  }]
}
constraints [
 {bone1Name=dyn_fur_01, bone2Name=Melee_TwoHanded_Handle, swing1LimitDegrees=20, swing2LimitDegrees=20, twistLimitDegrees=15}
 {bone1Name=dyn_fur_02, bone2Name=Melee_TwoHanded_Handle, swing1LimitDegrees=30, swing2LimitDegrees=20, twistLimitDegrees=15}
 {bone1Name=dyn_fur_03, bone2Name=Melee_TwoHanded_Handle, swing1LimitDegrees=30, swing2LimitDegrees=20, twistLimitDegrees=15}
 {bone1Name=dyn_fur_04, bone2Name=Melee_TwoHanded_Handle, swing1LimitDegrees=20, swing2LimitDegrees=20, twistLimitDegrees=15}
]
